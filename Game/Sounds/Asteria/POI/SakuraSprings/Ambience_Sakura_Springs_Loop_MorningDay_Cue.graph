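# Looping outdoor morning/day ambience cue for the Sakura Springs POI, using the Athena ambient attenuation and outdoor sound class.
# A quality-level node routes the first two levels to quad waves and the last two to stereo.
# Each branch picks one of three equally weighted looping day waves at random.
# An endlessly looping 20-second envelope gently dips and restores the volume, adding slow movement to each bed.
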SoundCue /Game/Sounds/Asteria/POI/SakuraSprings/Ambience_Sakura_Springs_Loop_MorningDay_Cue
root=QualityLevel_0
nodes [
 {
  id=Enveloper_2
  children [Random_0]
  properties {
   bLoop=true
   bLoopIndefinitely=true
   loopEnd=20
   pitchCurve {editorCurveData {keys [{time=0}]}}
   volumeCurve {editorCurveData {keys [
      {arriveTangent=-0.01, interpMode=RCIM_Cubic, leaveTangent=-0.01, time=0, value=1}
      <at 10 0.9 RCIM_Cubic>
      {arriveTangent=0.01, interpMode=RCIM_Cubic, leaveTangent=0.01, time=20, value=1}
     ]}}
  }
 }
 {
  id=Enveloper_3
  children [Random_1]
  properties {
   bLoop=true
   bLoopIndefinitely=true
   loopEnd=20
   pitchCurve {editorCurveData {keys [{time=0}]}}
   volumeCurve {editorCurveData {keys [
      {arriveTangent=-0.01, interpMode=RCIM_Cubic, leaveTangent=-0.01, time=0, value=1}
      <at 10 0.9 RCIM_Cubic>
      {arriveTangent=0.01, interpMode=RCIM_Cubic, leaveTangent=0.01, time=20, value=1}
     ]}}
  }
 }
 {id=QualityLevel_0, children [Enveloper_3,Enveloper_3,Enveloper_2,Enveloper_2]}
 {id=Random_0, children [WavePlayer_0,WavePlayer_1,WavePlayer_2], properties {weights [1,1,1]}}
 {id=Random_1, children [WavePlayer_3,WavePlayer_4,WavePlayer_5], properties {weights [1,1,1]}}
 {
  id=WavePlayer_0
  properties {bLooping=true, soundWaveAssetPtr=&/Ambience_Sakura_Springs_Ext_Day_1_Stereo}
 }
 {
  id=WavePlayer_1
  properties {bLooping=true, soundWaveAssetPtr=&/Ambience_Sakura_Springs_Ext_Day_2_Stereo}
 }
 {
  id=WavePlayer_2
  properties {bLooping=true, soundWaveAssetPtr=&/Ambience_Sakura_Springs_Ext_Day_3_Stereo}
 }
 {id=WavePlayer_3, properties {bLooping=true, soundWaveAssetPtr=&/Ambience_Sakura_Springs_Ext_Day_1_Quad}}
 {id=WavePlayer_4, properties {bLooping=true, soundWaveAssetPtr=&/Ambience_Sakura_Springs_Ext_Day_2_Quad}}
 {id=WavePlayer_5, properties {bLooping=true, soundWaveAssetPtr=&/Ambience_Sakura_Springs_Ext_Day_3_Quad}}
]
properties {
 attenuationSettings=&Fort_Ambient_Athena_Att
 maxDistance=2097152
 soundClassObject=&Fort_Ambient_Outdoors_Athena
 volumeMultiplier=0.601
 preEffectBusSends [{}]
}
